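# Physics asset for the Legs vehicle body cosmetic, supplying its collision shape.
# It holds a single body on the root bone, made of one box primitive offset forward and upward from the bone origin.
# It defines no constraints, so the whole mesh is represented physically as one block rather than articulated parts.
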PhysicsAsset /VehicleCosmetics/Bodies/Legs/Meshes/Body_Legs_PhysicsDM
bodies {root [{shapeType=Box, extentX=444, extentY=216, extentZ=132, shapeName=root_box, center=<32 0 97>}]}
constraints []
